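# Material function for the Oubliette universal vehicle skin, outputting material attributes with Octane body textures as defaults.
# An RGB mask supplies color mask, panel lines and environment mask, combined with team, custom and paint colors.
# Offset samples of panel lines and thresholded curvature edges are differenced into normals.
# Bump-offset curvature echoes and bevel glow, Fresnel and a time-driven sine pulse shape color and emissive, alongside a fake reflection-vector environment layer.
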MaterialFunction /VehicleCosmetics/UniversalSkins/Oubliette/MF_Oubliette
graph {nodes [
  n1 = NamedRerouteUsage()
  n2 = SetMaterialAttributes!(
   MaterialAttributes
   BaseColor=NamedRerouteUsage()
   Metallic=ScalarParameter() {parameterName=Metallic}
   Roughness=ScalarParameter() {defaultValue=0.2, parameterName=Roughness}
   EmissiveColor=NamedRerouteUsage()
   Normal=NamedRerouteUsage()
   AmbientOcclusion
   Specular
   TransmittanceColor=@n1
   SurfaceThickness=@n1
  ) {attributeSetTypes [
    #1
    #2
    #3
    #4
    #5
    #6
    #7
    #8
    #9
   ]}
  FunctionOutput(None=@n2)
  MaterialFunctionCall(Invert Fresnel=StaticBool() {value=true}, Power=Constant() {r=1}) {materialFunction=&Fresnel_Function}
  n3 = TextureSampleParameter2D(UVs=Multiply(A=TextureCoordinate() {coordinateIndex=1}, B=ScalarParameter() {defaultValue=1, parameterName=NormalTiling})) {parameterName=Normal, samplerType=SAMPLERTYPE_LinearColor, texture=/VehicleCosmetics/Textures/UniversalTextures/Matte_N}
  n4 = TextureCoordinate() {coordinateIndex=1}
  n5 = Constant() {desc=.04 * .085 from ue3 expressions, r=0.0034}
  n6 = Constant() {r=0.7}
  n7 = Multiply(A=@n5, B=@n6)
  n8 = Constant()
  n9 = TextureSampleParameter2D(UVs=Add(A=@n4, B=AppendVector(A=@n7, B=@n8))) {desc="R = colormask, G = PanelLines, B = env", parameterName=RGB, texture=&/Oubliette1_rgb}
  n10 = ScalarParameter() {defaultValue=1, parameterName=Curvature Edge Normal}
  n11 = TextureCoordinate()
  n12 = Constant() {r=0.0051}
  n13 = ScalarParameter() {defaultValue=0.7, parameterName=NormalOffset}
  n14 = Constant()
  n15 = Add(A=@n11, B=AppendVector(A=Multiply(A=@n12, B=@n13), B=@n14))
  n16 = ScalarParameter() {defaultValue=0.375, parameterName=Curvature Edge Sharpness}
  n17 = ScalarParameter() {defaultValue=500, parameterName=Curvature Edge Solid}
  n18 = Multiply(
   A=Subtract(
    A=TextureSampleParameter2D(UVs=@n15) {parameterName=Edge Mask, texture=/VehicleCosmetics/Bodies/Octane/Textures/Body/T_Octane_Body_Curvature_RGB}.R
    B=@n16
   )
   B=@n17
  )
  n19 = Max(A=LinearInterpolate(B=@n9.G, Alpha=@n9.G), B=Clamp(None=Multiply(A=@n10, B=@n18)) {maxDefault=0.2})
  n20 = Multiply(A=Multiply(A=@n5) {constB=-1}, B=@n6)
  n21 = TextureSampleParameter2D(UVs=Add(A=@n4, B=AppendVector(A=@n20, B=@n8))) {desc="R = colormask, G = PanelLines, B = env", parameterName=RGB, texture=&Oubliette1_rgb}
  n22 = Multiply(A=@n12) {constB=-1}
  n23 = Add(A=@n11, B=AppendVector(A=Multiply(A=@n22, B=@n13), B=@n14))
  n24 = Multiply(
   A=Subtract(
    A=TextureSampleParameter2D(UVs=@n23) {parameterName=Edge Mask, texture=&T_Octane_Body_Curvature_RGB}.R
    B=@n16
   )
   B=@n17
  )
  n25 = Max(A=LinearInterpolate(B=@n21.G, Alpha=@n21.G), B=Clamp(None=Multiply(A=@n10, B=@n24)) {maxDefault=0.3})
  n26 = TextureSampleParameter2D(UVs=@n4) {desc="R = colormask, G = PanelLines, B = env", parameterName=RGB, texture=&Oubliette1_rgb}
  n27 = TextureSampleParameter2D(UVs=@n11) {parameterName=Edge Mask, texture=&T_Octane_Body_Curvature_RGB}
  n28 = Multiply(A=@n10, B=Multiply(A=Subtract(A=@n27.R, B=@n16), B=@n17))
  n29 = Max(A=LinearInterpolate(B=@n26.G, Alpha=@n26.G), B=Clamp(None=@n28) {maxDefault=0.3})
  n30 = Max(A=Max(A=@n19, B=@n25), B=@n29)
  n31 = LinearInterpolate(
   A=Multiply(A=AppendVector(A=@n3.R, B=@n3.R), B=ScalarParameter() {defaultValue=0.05, parameterName=DetailNormalIntensity})
   B=Constant2Vector()
   Alpha=@n30
  ) {constB=0}
  n32 = Constant()
  n33 = Constant4Vector() {constant=<rgba 1 0 0 0>}
  n34 = DotProduct(A=@n19, B=@n33.RGBA)
  n35 = DotProduct(A=@n29, B=@n33.RGBA)
  n36 = Constant() {desc=NormalIntensity, r=1}
  n37 = AppendVector(A=Constant2Vector() {r=1}, B=Multiply(A=Subtract(A=@n34, B=@n35), B=@n36))
  n38 = DotProduct(A=@n25, B=@n33.RGBA)
  n39 = AppendVector(A=Constant2Vector() {g=1}, B=Multiply(A=Subtract(A=@n38, B=@n35), B=@n36))
  NamedRerouteDeclaration(None=Add(A=AppendVector(A=@n31, B=@n32), B=CrossProduct(A=@n37, B=@n39))) {name=Normal, nodeColor=<rgba 0 1 0.4375 1>}
  n40 = ComponentMask(None=ReflectionVectorWS()) {g=true, r=true}
  n41 = Constant() {r=1}
  n42 = TextureSample(UVs=Multiply(A=BumpOffset(Coordinate=@n40, Height=@n41) {heightRatio=1}) {constB=0.5}) {texture=/VehicleCosmetics/Textures/UniversalTextures/T_ENVPack}
  n43 = Multiply(
   A=Add(A=LinearInterpolate(A=@n42.R, B=@n42.R) {constAlpha=0.7}, B=ScalarParameter() {parameterName=INT Env Add})
   B=ScalarParameter() {parameterName=INT Env}
  )
  n44 = NamedRerouteDeclaration(None=LinearInterpolate(A=NamedRerouteUsage(), B=NamedRerouteUsage(), Alpha=@n26.R)) {name=Specular, nodeColor=<rgba 0 1 0.6953 1>}
  n45 = OneMinus(None=MaterialFunctionCall!(Fresnel Power=Constant() {r=0.2}, Normal=Constant3Vector() {constant=<rgba 0 0 1 0>}) {materialFunction=/VehicleCosmetics/Materials/MaterialFunctions/MF_Fresnel_Prime}.Result)
  n46 = LinearInterpolate(A=@n44, Alpha=Max(A=Multiply(A=@n42.G) {constB=0.1}, B=Multiply(A=@n45) {constB=0.8}))
  n47 = Power(
   Base=LinearInterpolate(A=@n43, B=Multiply(A=@n43, B=@n46), Alpha=ScalarParameter() {parameterName=INT EnvTint})
   Exp=ScalarParameter() {defaultValue=0.52, parameterName=INT ENV Sharpness}
  )
  n48 = MaterialFunctionCall(Fresnel Power=Constant() {r=1}, Normal=NamedRerouteUsage()) {materialFunction=&MF_Fresnel_Prime}
  n49 = LinearInterpolate(A=@n48.Result, B=OneMinus(None=@n48.Result), Alpha=ScalarParameter() {defaultValue=1, parameterName=INT Env FresnelInvert})
  NamedRerouteDeclaration(None=LinearInterpolate(B=Multiply(A=@n47, B=@n49), Alpha=@n26.B)) {name=Env, nodeColor=<rgba 1 0 0.2578 1>}
  n50 = VectorParameter() {desc="Alpha = Emissive Color Brightness", group=Colors, parameterName=BaseColor, defaultValue=<rgba 1 0 0 1>}
  n51 = ScalarParameter() {parameterName=Override Colors}
  NamedRerouteDeclaration(None=LinearInterpolate(A=@n50.RGB, B=VectorParameter() {parameterName=Override TeamColor, defaultValue=<rgba 1 0.159 0 1>}.RGB, Alpha=@n51)) {name=TeamColor, nodeColor=<rgba 1 0.0179 0 1>}
  NamedRerouteDeclaration(None=LinearInterpolate(
    A=VectorParameter() {group=Colors, parameterName=AccentColor, defaultValue=<rgba 0 0 1 1>}.RGB
    B=VectorParameter() {parameterName=Override CustomColor, defaultValue=<rgba 0 0.15 1 1>}.RGB
    Alpha=@n51
   )) {name=CustomColor, nodeColor=<rgba 0 0 1 1>}
  NamedRerouteDeclaration(None=LinearInterpolate(
    A=VectorParameter() {group=Colors, parameterName=PaintColor, defaultValue=<rgba 0 1 0.0137 1>}.RGB
    B=VectorParameter() {parameterName=Override PaintColor, defaultValue=<rgba 0.0774 0.213 1 1>}.RGB
    Alpha=@n51
   )) {name=PaintColor, nodeColor=<rgba 0 1 0.0199 1>}
  NamedRerouteDeclaration(None=Time() {bOverride_Period=true, period=500}) {name=time, nodeColor=<rgba 1 0.9285 0 1>}
  n52 = MaterialFunctionCall(Fresnel Power=Constant() {r=0.5}, Normal=NamedRerouteUsage()) {materialFunction=&MF_Fresnel_Prime}
  n53 = Desaturation(None=@n44, Fraction=Saturate(None=Multiply(A=@n52.Result) {constB=0.5}))
  n54 = Desaturation(None=@n53, Fraction=Constant() {r=0.2})
  n55 = Constant() {r=0.5}
  n56 = TextureSampleParameter2D(UVs=BumpOffset(Coordinate=@n11, Height=Constant() {r=0.48}, HeightRatioInput=@n13)) {parameterName=CurvaturePack, texture=&T_Octane_Body_Curvature_RGB}
  n57 = Max(A=@n26.G, B=Saturate(None=@n28))
  n58 = Subtract(A=ConstantBiasScale(None=@n56.R) {bias=-0.1}, B=ConstantBiasScale(None=@n57) {bias=-0.39, scale=500})
  n59 = Subtract(A=@n56.R, B=@n16)
  n60 = Multiply(A=Saturate(None=Multiply(A=@n59, B=@n17)), B=ScalarParameter() {defaultValue=1, parameterName=EchoMultiply})
  n61 = Saturate(None=Multiply(A=Max(A=@n57, B=@n60), B=ScalarParameter() {defaultValue=1, parameterName=CurvatureEdgeColor}))
  n62 = LinearInterpolate(A=NamedRerouteUsage(), B=LinearInterpolate(A=@n53, B=Desaturation(None=@n53, Fraction=@n55), Alpha=Saturate(None=@n58)), Alpha=@n61)
  n63 = MaterialFunctionCall(Fresnel Power=Constant() {r=3}, Normal=NamedRerouteUsage()) {materialFunction=&MF_Fresnel_Prime}
  n64 = Sine(None=Multiply(A=NamedRerouteUsage(), B=ScalarParameter() {defaultValue=1, parameterName=PulseTiming})) {period=8}
  n65 = Saturate(None=Add(A=Abs(None=@n64)) {constB=0.5})
  n66 = LinearInterpolate(B=Power(Base=OneMinus(None=@n63.Result)) {constExponent=5}, Alpha=@n65) {constA=1}
  NamedRerouteDeclaration(None=Multiply(A=LinearInterpolate(A=@n54, B=@n62, Alpha=@n66)) {constB=0.5}) {name=BaseColor, nodeColor=<rgba 0.5 1 0 1>}
  n67 = MaterialFunctionCall(Fresnel Power=Constant() {r=10}, Normal=NamedRerouteUsage()) {materialFunction=&MF_Fresnel_Prime}
  n68 = Add(A=@n65) {constB=0.3}
  n69 = LinearInterpolate(A=NamedRerouteUsage(), B=Desaturation(None=@n54, Fraction=Constant() {r=0.5}), Alpha=Saturate(None=Multiply(A=@n67.Result, B=@n68)))
  n70 = TextureSampleParameter2D(UVs=BumpOffset(Coordinate=@n11, Height=Constant() {r=0.49}, HeightRatioInput=ScalarParameter() {defaultValue=1, parameterName=GlowOffset})) {parameterName=FX Masks, texture=/VehicleCosmetics/Bodies/Octane/Textures/Body/T_Octane_Bevel_RGB}
  n71 = Subtract(A=Multiply(A=ConstantBiasScale(None=@n70.R) {bias=-0.6, scale=1}, B=ScalarParameter() {defaultValue=1, parameterName=Glow}), B=@n61)
  n72 = Saturate(None=Multiply(A=@n68, B=Subtract(A=@n71, B=@n30)))
  NamedRerouteDeclaration(None=Multiply(A=LinearInterpolate(A=@n69, B=@n53, Alpha=@n72), B=ScalarParameter() {parameterName=EmissiveIntensity})) {name=Emissive, nodeColor=<rgba 1 0.9375 0 1>}
  TextureSampleParameter2D() {parameterName=Edge Mask, texture=&T_Octane_Body_Curvature_RGB}
  TextureSampleParameter2D() {parameterName=CurvaturePack, texture=&T_Octane_Body_Curvature_RGB}
  TextureSampleParameter2D() {parameterName=FX Masks, texture=&T_Octane_Bevel_RGB}
  TextureSampleParameter2D() {parameterName=RGB, texture=&Oubliette1_rgb}
 ]}
